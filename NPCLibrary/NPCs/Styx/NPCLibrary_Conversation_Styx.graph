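# Conversation database for the NPC Styx, providing greeting dialogue and access to services and quests.
# The Styx entry point leads to a single speech node with a default merchant line and priority-ranked contextual lines.
# Contextual lines cover a first-ever interaction and reactions to specific player outfits, checked by character ID requirements.
# The speech node then branches through link nodes into the shared NPC library quest and basic services conversations.
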ConversationDatabase /NPCLibrary/NPCs/Styx/NPCLibrary_Conversation_Styx
entry_points [{tag=Conversation.Entry.Styx, destinations [#1]}]
nodes [
 ConversationEntryPointNode_7 {
  id=#1
  class=ConversationEntryPointNode
  outputs [#2]
  properties {entryTag {tagName=Conversation.Entry.Styx}}
 }
 Speech_C_4 {
  id=#2
  class=Speech_C
  outputs [#3, #4]
  properties {
   nodeName=""
   generalConfig {
    defaultMessage="I've got merchandise for sale. Don't ask where we \"bought\" it from."
    contextualMessages [
     {
      message=Dahlia told me to lay low. So... I'm laying low.
      contextRequirements [{requirement {class=FortControllerRequirement_IsFirstEverConversationWithNPC}}]
      priority {curve {rowName=Tandem.Speech.Priority.FirstEverInteraction}}
     }
     {
      message=We ready to make our move yet?
      contextRequirements [{requirement {class=FortControllerRequirement_HasCID, properties {allowedCIds [/BRCosmetics/Athena/Items/Cosmetics/Characters/Character_Lilac]}}}]
      priority {curve {rowName=Tandem.Speech.Priority.OutfitReaction}}
     }
     {
      message="Someday, buddy, you gotta tell me where you're from."
      contextRequirements [{requirement {class=FortControllerRequirement_HasCID, properties {allowedCIds [/BRCosmetics/Athena/Items/Cosmetics/Characters/Character_Titanium]}}}]
      priority {curve {rowName=Tandem.Speech.Priority.OutfitReaction}}
     }
     {
      message="You know I only listen to Dahlia, right?"
      contextRequirements [{requirement {class=FortControllerRequirement_HasCID, properties {allowedCIds [&Character_VitalPsych]}}}]
      priority {curve {rowName=Tandem.Speech.Priority.OutfitReaction}}
     }
     {
      message="Man, you've got the right attitude. Destroying stuff is FUN."
      contextRequirements [{requirement {class=FortControllerRequirement_HasCID, properties {allowedCIds [&Character_Inferno]}}}]
      priority {curve {rowName=Tandem.Speech.Priority.OutfitReaction}}
     }
     {
      message=I'm not really interested in peace. I'd rather leave stuff in pieces.
      contextRequirements [{requirement {class=FortControllerRequirement_HasCID, properties {allowedCIds [&Character_LocalZilla]}}}]
      priority {curve {rowName=Tandem.Speech.Priority.OutfitReaction}}
     }
    ]
   }
  }
 }
 ConversationLinkNode_0 {id=#4, class=ConversationLinkNode, properties {remoteEntryTag {tagName=Conversation.Entry.BasicServices}}}
 ConversationLinkNode_1 {id=#3, class=ConversationLinkNode, properties {remoteEntryTag {tagName=Conversation.Entry.NPCLibrary.Quests}}}
]
